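# Control Rig for the EmberYearn helmet face accessory on the female medium player body, repositioning hair bones to follow the body.
# On forward solve, a three-way sequence runs chains of parent constraints for back hair, left-side hair, then right-side and front hair.
# Each constraint blends weighted head, neck, spine, clavicle or arm bones, resembling skin weights.
# Many targeted mid and side hair bones are absent from this rig's hierarchy.
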ControlRigBlueprint /BRCosmetics/Characters/Player/Female/Medium/Bodies/F_MED_EmberYearnHelmet/Meshes/Parts/F_MED_EmberYearnHelmet_FaceAcc_CtrlRig
graphs [
 RigVMFunctionLibrary {links [], nodes []}
 RigVMFunction_Sequence_1_ContainedGraph {nodes [
   n1 = RigVMFunctionEntryNode()
   RigVMFunction_Sequence = &RigVMFunction_Sequence(ExecuteContext=@n1.ExecuteContext)
   n2 = &RigVMFunction_Sequence(ExecuteContext=@RigVMFunction_Sequence.B)
   RigVMFunctionReturnNode(A=@RigVMFunction_Sequence.A, B=@n2.A, C=@n2.B)
  ]}
 RigVMModel {nodes [
   RigUnit_BeginExecution = &RigUnit_BeginExecution()
   RigVMFunction_Sequence = RigVMAggregateNode(ExecuteContext=@RigUnit_BeginExecution.ExecutePin)
   &RigUnit_ParentConstraint(
    AdvancedSettings=(RotationOrderForFilter=XZY)
    Child=(Type=Bone,Name="dyn_Hair_BackL_03")
    Parents=((Item=(Type=Bone,Name="spine_05"),Weight=0.9),(Item=(Type=Bone,Name="neck_01"),Weight=0.1))
    Weight=1
    ExecutePin=@RigVMFunction_Sequence.A
   )
   > &RigUnit_ParentConstraint(
    AdvancedSettings=(RotationOrderForFilter=XZY)
    Child=(Type=Bone,Name="base_Hair_Back_03")
    Parents=((Item=(Type=Bone,Name="spine_05"),Weight=0.9),(Item=(Type=Bone,Name="neck_01"),Weight=0.1))
    Weight=1
   )
   > &RigUnit_ParentConstraint(
    AdvancedSettings=(RotationOrderForFilter=XZY)
    Child=(Type=Bone,Name="dyn_Hair_BackR_03")
    Parents=((Item=(Type=Bone,Name="spine_05"),Weight=0.9),(Item=(Type=Bone,Name="neck_01"),Weight=0.1))
    Weight=1
   )
   &RigUnit_ParentConstraint(
    AdvancedSettings=(RotationOrderForFilter=XZY)
    Child=(Type=Bone,Name="dyn_Hair_mid_03_l")
    Parents=((Item=(Type=Bone,Name="head"),Weight=1.0))
    Weight=1
    ExecutePin=@RigVMFunction_Sequence.B
   )
   > &RigUnit_ParentConstraint(
    AdvancedSettings=(RotationOrderForFilter=XZY)
    Child=(Type=Bone,Name="dyn_Hair_mid_04_l")
    Parents=((Item=(Type=Bone,Name="spine_04"),Weight=0.0242),(Item=(Type=Bone,Name="clavicle_l"),Weight=0.2992),(Item=(Type=Bone,Name="spine_05"),Weight=0.6766))
    Weight=1
   )
   > &RigUnit_ParentConstraint(
    AdvancedSettings=(RotationOrderForFilter=XZY)
    Child=(Type=Bone,Name="dyn_Hair_mid_05_l")
    Parents=((Item=(Type=Bone,Name="spine_04"),Weight=0.0669),(Item=(Type=Bone,Name="clavicle_l"),Weight=0.1177),(Item=(Type=Bone,Name="spine_05"),Weight=0.8154))
    Weight=1
   )
   > &RigUnit_ParentConstraint(
    AdvancedSettings=(RotationOrderForFilter=XZY)
    Child=(Type=Bone,Name="dyn_Hair_02_l")
    Parents=((Item=(Type=Bone,Name="head"),Weight=1.0))
    Weight=1
   )
   > &RigUnit_ParentConstraint(
    AdvancedSettings=(RotationOrderForFilter=XZY)
    Child=(Type=Bone,Name="dyn_Hair_03_l")
    Parents=((Item=(Type=Bone,Name="neck_01"),Weight=0.1663),(Item=(Type=Bone,Name="head"),Weight=0.3674),(Item=(Type=Bone,Name="neck_02"),Weight=0.4663))
    Weight=1
   )
   > &RigUnit_ParentConstraint(
    AdvancedSettings=(RotationOrderForFilter=XZY)
    Child=(Type=Bone,Name="dyn_Hair_04_l")
    Parents=((Item=(Type=Bone,Name="upperarm_twist_01_l"),Weight=0.0428),(Item=(Type=Bone,Name="spine_04"),Weight=0.0491),(Item=(Type=Bone,Name="spine_05"),Weight=0.2978),(Item=(Type=Bone,Name="clavicle_l"),Weight=0.6103))
    Weight=1
   )
   &RigUnit_ParentConstraint(
    AdvancedSettings=(RotationOrderForFilter=XZY)
    Child=(Type=Bone,Name="dyn_Hair_front_mid_01_r")
    Parents=((Item=(Type=Bone,Name="head"),Weight=1.0))
    Weight=1
    ExecutePin=@RigVMFunction_Sequence.C
   )
   > &RigUnit_ParentConstraint(
    AdvancedSettings=(RotationOrderForFilter=XZY)
    Child=(Type=Bone,Name="dyn_Hair_front_mid_02_r")
    Parents=((Item=(Type=Bone,Name="head"),Weight=0.0528),(Item=(Type=Bone,Name="neck_01"),Weight=0.1313),(Item=(Type=Bone,Name="neck_02"),Weight=0.1777),(Item=(Type=Bone,Name="spine_05"),Weight=0.6382))
    Weight=1
   )
   > &RigUnit_ParentConstraint(
    AdvancedSettings=(RotationOrderForFilter=XZY)
    Child=(Type=Bone,Name="dyn_Hair_front_mid_03_r")
    Parents=((Item=(Type=Bone,Name="clavicle_r"),Weight=0.2152),(Item=(Type=Bone,Name="spine_05"),Weight=0.7848))
    Weight=1
   )
   > &RigUnit_ParentConstraint(
    AdvancedSettings=(RotationOrderForFilter=XZY)
    Child=(Type=Bone,Name="dyn_Hair_03_r")
    Parents=((Item=(Type=Bone,Name="neck_01"),Weight=0.1662),(Item=(Type=Bone,Name="head"),Weight=0.3675),(Item=(Type=Bone,Name="neck_02"),Weight=0.4663))
    Weight=1
   )
   > &RigUnit_ParentConstraint(
    AdvancedSettings=(RotationOrderForFilter=XZY)
    Child=(Type=Bone,Name="dyn_Hair_04_r")
    Parents=((Item=(Type=Bone,Name="upperarm_twist_01_r"),Weight=0.0208),(Item=(Type=Bone,Name="spine_04"),Weight=0.0424),(Item=(Type=Bone,Name="clavicle_r"),Weight=0.4646),(Item=(Type=Bone,Name="spine_05"),Weight=0.4722))
    Weight=1
   )
   > &RigUnit_ParentConstraint(
    AdvancedSettings=(RotationOrderForFilter=XZY)
    Child=(Type=Bone,Name="dyn_front_hair_r_01")
    Parents=((Item=(Type=Bone,Name="clavicle_r"),Weight=0.0142),(Item=(Type=Bone,Name="spine_04"),Weight=0.0323),(Item=(Type=Bone,Name="neck_01"),Weight=0.2245),(Item=(Type=Bone,Name="neck_02"),Weight=0.3036),(Item=(Type=Bone,Name="spine_05"),Weight=0.4255))
    Weight=1
   )
   > &RigUnit_ParentConstraint(
    AdvancedSettings=(RotationOrderForFilter=XZY)
    Child=(Type=Bone,Name="dyn_front_hair_r_02")
    Parents="((Item=(Type=Bone,Name=\"spine_03\"),Weight=0.0043),(Item=(Type=Bone,Name=\"neck_02\"),Weight=0.0045),(Item=(Type=Bone,Name=\"neck_01\"),Weight=0.0147),(Item=(Type=Bone,Name=\"spine_04\"),Weight=0.0962),(Item=(Type=Bone,Name=\"clavicle_r\"),Weight=0.251),(Item=(Type ...(truncated)"
    Weight=1
   )
   > &RigUnit_ParentConstraint(
    AdvancedSettings=(RotationOrderForFilter=XZY)
    Child=(Type=Bone,Name="dyn_front_hair_l_01")
    Parents="((Item=(Type=Bone,Name=\"spine_03\"),Weight=0.0031),(Item=(Type=Bone,Name=\"clavicle_l\"),Weight=0.0401),(Item=(Type=Bone,Name=\"spine_04\"),Weight=0.0414),(Item=(Type=Bone,Name=\"neck_01\"),Weight=0.1967),(Item=(Type=Bone,Name=\"neck_02\"),Weight=0.2053),(Item=(Typ ...(truncated)"
    Weight=1
   )
   > &RigUnit_ParentConstraint(
    AdvancedSettings=(RotationOrderForFilter=XZY)
    Child=(Type=Bone,Name="dyn_front_hair_l_02")
    Parents="((Item=(Type=Bone,Name=\"deform_upperarm_fr_l\"),Weight=0.0089),(Item=(Type=Bone,Name=\"deform_pec_l\"),Weight=0.0222),(Item=(Type=Bone,Name=\"spine_04\"),Weight=0.1229),(Item=(Type=Bone,Name=\"clavicle_l\"),Weight=0.2877),(Item=(Type=Bone,Name=\"spine_05\"),Weight= ...(truncated)"
    Weight=1
   )
  ]}
]
hierarchy [
 root {loc=<0 0 0>, rot=<0 0 0 1>}
 pelvis {parent=root, loc=<0 -0.5301 98.0571>, rot=<0.0127 0.707 -0.0127 -0.707>}
 spine_01 {parent=pelvis, loc=<6.1208 -0.9328 0>, rot=<0 0 0.034 -0.9994>}
 spine_02 {parent=spine_01, loc=<6.5997 -0.2095 0>, rot=<0 0 0.0238 -0.9997>}
 spine_03 {parent=spine_02, loc=<6.4914 -0.1237 0>, rot=<0 0 0.0149 -0.9999>}
 spine_04 {parent=spine_03, loc=<6.9598 0.0794 0>, rot=<0 0 -0.0132 -0.9999>}
 spine_05 {parent=spine_04, loc=<6.9987 0 0>, rot=<0 0 -0.0015 -1>}
 neck_01 {parent=spine_05, loc=<11.3138 0.1529 0>, rot=<0 0 -0.2003 -0.9797>}
 neck_02 {parent=neck_01, loc=<4.5627 0 0>, rot=<0 0 0.0504 -0.9987>}
 head {parent=neck_02, loc=<5.3914 0 0>, rot=<0 0 0.115 -0.9934>}
 base_Hair_Back {parent=head, loc=<8.5052 -1.7042 -0.3258>, rot=<0.3011 -0.6398 0.6398 -0.3011>}
 dyn_Hair_Back_01 {parent=base_Hair_Back, loc=<0 0 0>, rot=<0 0 0 1>}
 dyn_Hair_Back_02 {parent=dyn_Hair_Back_01, loc=<4.2813 0 0>, rot=<0 -0.3651 0 -0.931>}
 base_Hair_Back_03 {parent=dyn_Hair_Back_02, loc=<6.9171 0 0>, rot=<0.0034 0.1783 -0.019 -0.9838>}
 dyn_Hair_Back_03 {parent=base_Hair_Back_03, loc=<0 0 0>, rot=<0 0 0 1>}
 dyn_Hair_Back_04 {parent=dyn_Hair_Back_03, loc=<9.1968 0 0>, rot=<-0.0011 -0.0561 0.0193 -0.9982>}
 dyn_Hair_Back_05 {parent=dyn_Hair_Back_04, loc=<10.1314 0 0>, rot=<0 -0.2093 0 -0.9779>}
 base_Hair_BackR {parent=head, loc=<8.1891 -0.8779 5.0959>, rot=<0.3011 -0.6398 0.6398 -0.3011>}
 dyn_Hair_BackR_01 {parent=base_Hair_BackR, loc=<0 0 0>, rot=<0 0 0 1>}
 dyn_Hair_BackR_02 {parent=dyn_Hair_BackR_01, loc=<4.4274 0 0.1575>, rot=<0.0018 -0.3557 0.0046 -0.9346>}
 base_Hair_BackR_03 {parent=dyn_Hair_BackR_02, loc=<6.7978 0 0>, rot=<-0.0013 0.1616 -0.0016 -0.9869>}
 dyn_Hair_BackR_03 {parent=base_Hair_BackR_03, loc=<0 0 0>, rot=<0 0 0 1>}
 dyn_Hair_BackR_04 {parent=dyn_Hair_BackR_03, loc=<9.2816 0 0>, rot=<0.0001 -0.033 -0.0043 -0.9994>}
 dyn_Hair_BackR_05 {parent=dyn_Hair_BackR_04, loc=<7.3337 0 0>, rot=<-0.0002 -0.142 -0.0003 -0.9899>}
 base_Hair_BackL {parent=head, loc=<8.1893 -0.8779 -5.0959>, rot=<0.3011 -0.6398 0.6398 -0.3011>}
 dyn_Hair_BackL_01 {parent=base_Hair_BackL, loc=<0 0 0>, rot=<0 0.0178 0 -0.9998>}
 dyn_Hair_BackL_02 {parent=dyn_Hair_BackL_01, loc=<4.4301 0 0>, rot=<-0.0017 -0.3721 -0.0041 -0.9282>}
 base_Hair_BackL_03 {parent=dyn_Hair_BackL_02, loc=<6.799 0 0>, rot=<0.0015 0.1503 -0.0011 -0.9886>}
 dyn_Hair_BackL_03 {parent=base_Hair_BackL_03, loc=<0 0 0>, rot=<0 0 0 1>}
 dyn_Hair_BackL_04 {parent=dyn_Hair_BackL_03, loc=<9.2764 0 0>, rot=<-0.0002 -0.0328 0.0055 -0.9994>}
 dyn_Hair_BackL_05 {parent=dyn_Hair_BackL_04, loc=<7.291 0 0>, rot=<0 -0.1359 0 -0.9907>}
 dyn_helmet_offset_01 {parent=head, loc=<0 0 0>, rot=<0 0 0 1>}
 base_front_hair_r_01 {parent=head, loc=<-8.2628 0.4574 6.2476>, rot=<-0.1244 -0.188 0.9181 0.3261>}
 dyn_front_hair_r_01 {parent=base_front_hair_r_01, loc=<0 0 0>, rot=<0 0 0 1>}
 dyn_front_hair_r_02 {parent=dyn_front_hair_r_01, loc=<4.3534 0 0>, rot=<-0.1445 -0.0496 0.0072 -0.9882>}
 base_front_hair_l_01 {parent=head, loc=<-8.7339 0.4453 -8.0421>, rot=<-0.0011 0.0321 0.9529 0.3017>}
 dyn_front_hair_l_01 {parent=base_front_hair_l_01, loc=<0 0 0>, rot=<0 0 0 1>}
 dyn_front_hair_l_02 {parent=dyn_front_hair_l_01, loc=<5.8792 0 0.3281>, rot=<0 0 0 1>}
 base_zipper_01 {parent=neck_01, loc=<2.0011 6.2061 0>, rot=<-0.5178 -0.4413 -0.469 -0.5632>}
 dyn_zipper_01 {parent=base_zipper_01, loc=<0 0 0>, rot=<0 0 0 1>}
 dyn_zipper_02 {parent=dyn_zipper_01, loc=<0.4051 0.0218 -0.0738>, rot=<0 0 0 1>}
 clavicle_l {parent=spine_05, loc=<5.4831 -0.1053 -3.2946>, rot=<0.0465 -0.7289 -0.009 -0.6829>}
 upperarm_l {parent=clavicle_l, loc=<9.8658 -1.2136 -0.4105>, rot=<0.0164 -0.3245 -0.0269 -0.9454>}
 lowerarm_l {parent=upperarm_l, loc=<26.465 -0.8345 0.259>, rot=<0.0036 0.0161 -0.2298 -0.9731>}
 hand_l {parent=lowerarm_l, loc=<22.7355 -0.3203 -0.6105>, rot=<-0.6784 -0.0223 0.0735 -0.7307>}
 hand_attach_l {parent=hand_l, loc=<8.5 -3 0>, rot=<0.7071 0.7071 0 0>}
 weapon_l {parent=hand_l, loc=<-0.1974 -3.4111 0.3811>, rot=<0 0 0 1>}
 thumb_01_l {parent=hand_l, loc=<1.3118 -0.3191 -2.459>, rot=<0.6546 -0.4042 0.0664 -0.6354>}
 thumb_02_l {parent=thumb_01_l, loc=<5.9336 -0.0448 0.0401>, rot=<-0.0037 0.0213 0.298 -0.9543>}
 thumb_03_l {parent=thumb_02_l, loc=<2.2223 0.3682 -0.0907>, rot=<0.0118 -0.013 0.0066 -0.9998>}
 index_metacarpal_l {parent=hand_l, loc=<1.6535 0.0183 -2.96>, rot=<0.0158 -0.0124 -0.0173 -0.9996>}
 index_01_l {parent=index_metacarpal_l, loc=<8.3008 0.0357 -0.0524>, rot=<-0.0044 0.0485 0.0924 -0.9945>}
 index_02_l {parent=index_01_l, loc=<3.3587 -0.3004 0.0452>, rot=<-0.002 0.0027 0.3225 -0.9466>}
 index_03_l {parent=index_02_l, loc=<2.5578 -0.123 -0.0252>, rot=<-0.002 0.0067 0.2558 -0.9667>}
 middle_metacarpal_l {parent=hand_l, loc=<2.1856 -0.1725 -0.5984>, rot=<0.006 0.0168 -0.0329 -0.9993>}
 middle_01_l {parent=middle_metacarpal_l, loc=<7.3275 -0.2447 0.0129>, rot=<-0.0208 0.0042 0.0858 -0.9961>}
 middle_02_l {parent=middle_01_l, loc=<4.0123 0.1401 0.1175>, rot=<-0.0218 0.0258 0.3512 -0.9357>}
 middle_03_l {parent=middle_02_l, loc=<2.6077 0.0534 0.1373>, rot=<0.0082 0.0188 0.1708 -0.9851>}
 ring_metacarpal_l {parent=hand_l, loc=<2.2074 -0.3625 1.3638>, rot=<-0.0284 0.0722 -0.0458 -0.9959>}
 ring_01_l {parent=ring_metacarpal_l, loc=<7.0295 -0.6197 -0.1446>, rot=<-0.0489 -0.0252 0.1892 -0.9804>}
 ring_02_l {parent=ring_01_l, loc=<3.6174 0.4647 -0.1027>, rot=<0 0 0.2806 -0.9598>}
 ring_03_l {parent=ring_02_l, loc=<2.4152 0.114 0.0199>, rot=<0 0 0.1189 -0.9929>}
 pinky_metacarpal_l {parent=hand_l, loc=<1.7616 -0.9753 3.3814>, rot=<-0.0397 0.0814 -0.069 -0.9935>}
 pinky_01_l {parent=pinky_metacarpal_l, loc=<7.1709 -0.4831 -0.1094>, rot=<-0.0654 -0.0035 0.2242 -0.9723>}
 pinky_02_l {parent=pinky_01_l, loc=<2.8274 0.0893 -0.1043>, rot=<0.0037 0.0011 0.2854 -0.9584>}
 pinky_03_l {parent=pinky_02_l, loc=<2.2046 0.0307 -0.0078>, rot=<-0.0002 0.0098 0.1445 -0.9895>}
 deform_wrist_dn_base_l {parent=hand_l, loc=<0 -1.5326 0>, rot=<0 0 0 -1>}
 deform_wrist_dn_l {parent=deform_wrist_dn_base_l, loc=<0 0 0>, rot=<0 0 0 1>}
 deform_wrist_up_base_l {parent=hand_l, loc=<0 2.3009 0>, rot=<0 0 0 -1>}
 deform_wrist_up_l {parent=deform_wrist_up_base_l, loc=<0 0 0>, rot=<0 0 0 1>}
 lowerarm_twist_02_l {parent=lowerarm_l, loc=<10.6707 0 0>, rot=<0 0 0 1>}
 lowerarm_twist_01_l {parent=lowerarm_l, loc=<18.316 0 0>, rot=<0 0 0 1>}
 deform_elbow_in_base_l {parent=lowerarm_l, loc=<0.1834 3.6062 -0.2572>, rot=<0 0 0 -1>}
 deform_elbow_in_l {parent=deform_elbow_in_base_l, loc=<0 0 0>, rot=<0 0 0 1>}
 deform_elbow_out_base_l {parent=lowerarm_l, loc=<0 -4 0>, rot=<0 0 0 1>}
 deform_elbow_out_l {parent=deform_elbow_out_base_l, loc=<0 0 0>, rot=<0 0 0 1>}
 upperarm_twist_01_l {parent=upperarm_l, loc=<7.2135 -0.1134 0>, rot=<0 0 0 1>}
 deform_upperarm_bk_base_l {parent=upperarm_twist_01_l, loc=<-6.4053 -4.3746 0.314>, rot=<0 0 0 -1>}
 deform_upperarm_bk_l {parent=deform_upperarm_bk_base_l, loc=<0 0 0>, rot=<0 0 0 1>}
 deform_upperarm_fr_base_l {parent=upperarm_twist_01_l, loc=<-5.6312 5.092 0.3354>, rot=<0 0 0 -1>}
 deform_upperarm_fr_l {parent=deform_upperarm_fr_base_l, loc=<0 0 0>, rot=<0 0 0 1>}
 deform_upperarm_up_base_l {parent=upperarm_twist_01_l, loc=<-8.891 0.4825 3.9517>, rot=<0 0 0 -1>}
 deform_upperarm_up_l {parent=deform_upperarm_up_base_l, loc=<0 0 0>, rot=<0 0 0 1>}
 upperarm_twist_02_l {parent=upperarm_l, loc=<19.7998 -0.1134 0>, rot=<0 0 0 1>}
 clavicle_r {parent=spine_05, loc=<5.4831 -0.1053 3.2946>, rot=<0.7289 0.0465 -0.6829 0.009>}
 upperarm_r {parent=clavicle_r, loc=<-9.8658 1.2136 0.4105>, rot=<0.0164 -0.3245 -0.0269 -0.9454>}
 upperarm_twist_01_r {parent=upperarm_r, loc=<-7.2135 0.1134 0>, rot=<0 0 0 -1>}
 deform_upperarm_bk_base_r {parent=upperarm_twist_01_r, loc=<6.4053 4.3746 -0.314>, rot=<0 0 0 -1>}
 deform_upperarm_bk_r {parent=deform_upperarm_bk_base_r, loc=<0 0 0>, rot=<0 0 0 1>}
 deform_upperarm_fr_base_r {parent=upperarm_twist_01_r, loc=<5.6312 -5.092 -0.3354>, rot=<0 0 0 -1>}
 deform_upperarm_fr_r {parent=deform_upperarm_fr_base_r, loc=<0 0 0>, rot=<0 0 0 1>}
 deform_upperarm_up_base_r {parent=upperarm_twist_01_r, loc=<8.891 -0.4825 -3.9517>, rot=<0 0 0 -1>}
 deform_upperarm_up_r {parent=deform_upperarm_up_base_r, loc=<0 0 0>, rot=<0 0 0 1>}
 lowerarm_r {parent=upperarm_r, loc=<-26.465 0.8345 -0.259>, rot=<0.0036 0.0161 -0.2298 -0.9731>}
 hand_r {parent=lowerarm_r, loc=<-22.7355 0.3203 0.6106>, rot=<-0.6784 -0.0223 0.0735 -0.7307>}
 thumb_01_r {parent=hand_r, loc=<-1.3118 0.3191 2.459>, rot=<0.6546 -0.4042 0.0664 -0.6354>}
 thumb_02_r {parent=thumb_01_r, loc=<-5.9336 0.0448 -0.0401>, rot=<-0.0037 0.0213 0.298 -0.9543>}
 thumb_03_r {parent=thumb_02_r, loc=<-2.2223 -0.3682 0.0907>, rot=<0.0118 -0.013 0.0066 -0.9998>}
 index_metacarpal_r {parent=hand_r, loc=<-1.6535 -0.0183 2.96>, rot=<0.0158 -0.0124 -0.0173 -0.9996>}
 index_01_r {parent=index_metacarpal_r, loc=<-8.3008 -0.0357 0.0524>, rot=<-0.0044 0.0485 0.0924 -0.9945>}
 index_02_r {parent=index_01_r, loc=<-3.3587 0.3004 -0.0452>, rot=<-0.002 0.0027 0.3225 -0.9466>}
 index_03_r {parent=index_02_r, loc=<-2.5578 0.123 0.0252>, rot=<-0.002 0.0067 0.2558 -0.9667>}
 middle_metacarpal_r {parent=hand_r, loc=<-2.1856 0.1725 0.5984>, rot=<0.006 0.0168 -0.0329 -0.9993>}
 middle_01_r {parent=middle_metacarpal_r, loc=<-7.3275 0.2447 -0.0129>, rot=<-0.0208 0.0042 0.0858 -0.9961>}
 middle_02_r {parent=middle_01_r, loc=<-4.0123 -0.1401 -0.1175>, rot=<-0.0218 0.0258 0.3512 -0.9357>}
 middle_03_r {parent=middle_02_r, loc=<-2.6077 -0.0534 -0.1373>, rot=<0.0082 0.0188 0.1708 -0.9851>}
 ring_metacarpal_r {parent=hand_r, loc=<-2.2074 0.3625 -1.3638>, rot=<-0.0284 0.0722 -0.0458 -0.9959>}
 ring_01_r {parent=ring_metacarpal_r, loc=<-7.0295 0.6197 0.1446>, rot=<-0.0489 -0.0252 0.1892 -0.9804>}
 ring_02_r {parent=ring_01_r, loc=<-3.6174 -0.4647 0.1027>, rot=<0 0 0.2806 -0.9598>}
 ring_03_r {parent=ring_02_r, loc=<-2.4152 -0.114 -0.0199>, rot=<0 0 0.1189 -0.9929>}
 pinky_metacarpal_r {parent=hand_r, loc=<-1.7616 0.9753 -3.3814>, rot=<-0.0397 0.0814 -0.069 -0.9935>}
 pinky_01_r {parent=pinky_metacarpal_r, loc=<-7.1709 0.4831 0.1094>, rot=<-0.0654 -0.0035 0.2242 -0.9723>}
 pinky_02_r {parent=pinky_01_r, loc=<-2.8274 -0.0893 0.1043>, rot=<0.0037 0.0011 0.2854 -0.9584>}
 pinky_03_r {parent=pinky_02_r, loc=<-2.2046 -0.0307 0.0078>, rot=<-0.0002 0.0098 0.1445 -0.9895>}
 weapon_r {parent=hand_r, loc=<0.1974 3.4111 -0.3811>, rot=<0 0 0 1>}
 hand_attach_r {parent=hand_r, loc=<-8.5001 2.9999 0>, rot=<0 0 -0.7071 -0.7071>}
 deform_wrist_dn_base_r {parent=hand_r, loc=<-0.0001 1.5328 0>, rot=<0 0 0 -1>}
 deform_wrist_dn_r {parent=deform_wrist_dn_base_r, loc=<0 0 0>, rot=<0 0 0 1>}
 deform_wrist_up_base_r {parent=hand_r, loc=<-0.0002 -2.3007 -0.0001>, rot=<0 0 0 -1>}
 deform_wrist_up_r {parent=deform_wrist_up_base_r, loc=<0 0 0>, rot=<0 0 0 1>}
 lowerarm_twist_02_r {parent=lowerarm_r, loc=<-10.6707 0 0>, rot=<0 0 0 -1>}
 lowerarm_twist_01_r {parent=lowerarm_r, loc=<-18.316 0 0>, rot=<0 0 0 -1>}
 deform_elbow_in_base_r {parent=lowerarm_r, loc=<-0.1834 -3.6062 0.2572>, rot=<0 0 0 -1>}
 deform_elbow_in_r {parent=deform_elbow_in_base_r, loc=<0 0 0>, rot=<0 0 0 1>}
 deform_elbow_out_base_r {parent=lowerarm_r, loc=<0 4 -0.0001>, rot=<0 0 0 -1>}
 deform_elbow_out_r {parent=deform_elbow_out_base_r, loc=<0 0 0>, rot=<0 0 0 1>}
 upperarm_twist_02_r {parent=upperarm_r, loc=<-19.7998 0.1134 0>, rot=<0 0 0 -1>}
 attach_backpack {parent=spine_05, loc=<-0.123 -13.473 0>, rot=<-0.0093 -0.707 0.0093 -0.707>}
 attach_cape {parent=spine_05, loc=<9.017 -11.714 0>, rot=<0.0616 -0.7044 -0.0616 -0.7044>}
 aim_const {parent=spine_05, loc=<0 0 0>, rot=<0 0 0 1>}
 deform_pec_aim_base_l {parent=spine_04, loc=<3.2456 9.3309 -1.2237>, rot=<0.0603 -0.4033 0.3318 -0.8507>}
 deform_pec_aim_l {parent=deform_pec_aim_base_l, loc=<0 0 0>, rot=<0 0 0 1>}
 deform_pec_base_l {parent=deform_pec_aim_l, loc=<9.3665 0 0>, rot=<0 0 0 1>}
 deform_pec_l {parent=deform_pec_base_l, loc=<0 0 0>, rot=<0 0 0 1>}
 deform_pec_aim_base_r {parent=spine_04, loc=<3.2452 9.3309 1.2237>, rot=<-0.4033 -0.0603 0.8507 0.3318>}
 deform_pec_aim_r {parent=deform_pec_aim_base_r, loc=<0 0 0>, rot=<0 0 0 1>}
 deform_pec_base_r {parent=deform_pec_aim_r, loc=<-9.3667 -0.0003 -0.0004>, rot=<0 0 0 1>}
 deform_pec_r {parent=deform_pec_base_r, loc=<0 0 0>, rot=<0 0 0 1>}
 base_jacket_01_r {parent=spine_04, loc=<-0.7534 -1.0191 12.0594>, rot=<-0.1032 -0.0053 0.9941 -0.0319>}
 dyn_jacket_01_r {parent=base_jacket_01_r, loc=<0 0 0>, rot=<0 0 0 1>}
 dyn_jacket_02_r {parent=dyn_jacket_01_r, loc=<4.7125 0 0>, rot=<0 0 0 1>}
 base_jacket_01_f {parent=spine_04, loc=<-2.7212 13.1119 0>, rot=<0.0562 -0.7049 -0.7049 0.0562>}
 dyn_jacket_01_f {parent=base_jacket_01_f, loc=<0 0 0>, rot=<0 0 0 1>}
 dyn_jacket_02_f {parent=dyn_jacket_01_f, loc=<7.4631 0 0>, rot=<0 0 0 1>}
 base_jacket_01_l {parent=spine_04, loc=<-0.7534 -1.0191 -12.0594>, rot=<0.0321 -0.9941 0.0072 0.1031>}
 dyn_jacket_01_l {parent=base_jacket_01_l, loc=<0 0 0>, rot=<0 0 0 1>}
 dyn_jacket_02_l {parent=dyn_jacket_01_l, loc=<4.7128 0 0>, rot=<0 0 0 1>}
 base_jacket_01_b {parent=spine_04, loc=<0.0393 -9.8524 0>, rot=<-0.0154 -0.7069 0.7069 0.0154>}
 dyn_jacket_01_b {parent=base_jacket_01_b, loc=<0 0 0>, rot=<0 0 0 1>}
 dyn_jacket_02_b {parent=dyn_jacket_01_b, loc=<4.6899 0 0>, rot=<0 0 0 1>}
 thigh_l {parent=pelvis, loc=<-1.0855 -0.2109 -8.7053>, rot=<0.0009 0.0355 0.0155 -0.9993>}
 calf_l {parent=thigh_l, loc=<-42.5284 -0.0019 -0.7297>, rot=<0.0006 -0.0161 -0.0391 -0.9991>}
 foot_l {parent=calf_l, loc=<-46.7401 -2.5014 -0.592>, rot=<0.0427 -0.0238 0.0353 -0.9982>}
 ball_l {parent=foot_l, loc=<-3.5658 10.8066 -0.1544>, rot=<0.0001 -0.0001 -0.7063 -0.7079>}
 deform_foot_in_base_l {parent=foot_l, loc=<0 -3.5 0>, rot=<0 0 0 1>}
 deform_foot_in_l {parent=deform_foot_in_base_l, loc=<0 0 0>, rot=<0.0001 -0.0001 -0.7063 -0.7079>}
 calf_twist_01_l {parent=calf_l, loc=<-35.8483 -1.9197 -0.5794>, rot=<0.0237 0.0095 -0.0035 -0.9997>}
 calf_twist_02_l {parent=calf_l, loc=<-19.1546 -1.7943 -0.2638>, rot=<0.0237 0.0095 -0.0035 -0.9997>}
 deform_knee_in_base_l {parent=calf_l, loc=<0.8031 -5.0231 0.2456>, rot=<0 0 0 -1>}
 deform_knee_in_l {parent=deform_knee_in_base_l, loc=<0 0 0>, rot=<0 0 0 1>}
 deform_knee_out_base_l {parent=calf_l, loc=<0 4 0>, rot=<0 0 0 -1>}
 deform_knee_out_l {parent=deform_knee_out_base_l, loc=<0 0 0>, rot=<0 0 0 1>}
 thigh_twist_01_l {parent=thigh_l, loc=<-27.952 0 0>, rot=<0.0013 0 0 -1>}
 deform_groin_base_l {parent=thigh_l, loc=<0.4801 5.6776 -0.2486>, rot=<0 0 0 -1>}
 deform_groin_l {parent=deform_groin_base_l, loc=<0 0 0>, rot=<0 0 0 1>}
 thigh_r {parent=pelvis, loc=<-1.0855 -0.2109 8.7053>, rot=<0.0355 -0.0009 0.9993 0.0155>}
 calf_r {parent=thigh_r, loc=<42.5284 0.0019 0.7297>, rot=<0.0006 -0.0161 -0.0391 -0.9991>}
 foot_r {parent=calf_r, loc=<46.7401 2.5014 0.592>, rot=<0.0427 -0.0238 0.0353 -0.9982>}
 ball_r {parent=foot_r, loc=<3.5658 -10.8066 0.1544>, rot=<0.0001 -0.0001 -0.7063 -0.7079>}
 deform_foot_in_base_r {parent=foot_r, loc=<0 3.5 0>, rot=<0 0 0 1>}
 deform_foot_in_r {parent=deform_foot_in_base_r, loc=<0 0 0>, rot=<0.0001 -0.0001 -0.7063 -0.7079>}
 calf_twist_01_r {parent=calf_r, loc=<35.8483 1.9197 0.5794>, rot=<0.0237 0.0095 -0.0035 -0.9997>}
 calf_twist_02_r {parent=calf_r, loc=<19.1546 1.7943 0.2638>, rot=<0.0237 0.0095 -0.0035 -0.9997>}
 deform_knee_in_base_r {parent=calf_r, loc=<-0.8031 5.0231 -0.2457>, rot=<0 0 0 -1>}
 deform_knee_in_r {parent=deform_knee_in_base_r, loc=<0 0 0>, rot=<0 0 0 1>}
 deform_knee_out_base_r {parent=calf_r, loc=<0 -4 0>, rot=<0 0 0 -1>}
 deform_knee_out_r {parent=deform_knee_out_base_r, loc=<0 0 0>, rot=<0 0 0 1>}
 thigh_twist_01_r {parent=thigh_r, loc=<27.952 0 0>, rot=<0.0013 0 0 -1>}
 deform_groin_base_r {parent=thigh_r, loc=<-0.4801 -5.6776 0.2486>, rot=<0 0 0 -1>}
 deform_groin_r {parent=deform_groin_base_r, loc=<0 0 0>, rot=<0 0 0 1>}
 deform_glute_base_l {parent=pelvis, loc=<-7.3466 -9.0213 -9.1245>, rot=<0.0172 0.031 -0.4477 -0.8935>}
 deform_glute_l {parent=deform_glute_base_l, loc=<0 0 0>, rot=<0 0 0 1>}
 deform_glute_base_r {parent=pelvis, loc=<-7.3464 -9.0213 9.1245>, rot=<0.031 -0.0172 0.8935 -0.4477>}
 deform_glute_r {parent=deform_glute_base_r, loc=<0 0 0>, rot=<0 0 0 1>}
 conform {type=Curve, loc=<0 0 0>, rot=<0 0 0 1>}
 widenZ {type=Curve, loc=<0 0 0>, rot=<0 0 0 1>}
 widenX {type=Curve, loc=<0 0 0>, rot=<0 0 0 1>}
]
