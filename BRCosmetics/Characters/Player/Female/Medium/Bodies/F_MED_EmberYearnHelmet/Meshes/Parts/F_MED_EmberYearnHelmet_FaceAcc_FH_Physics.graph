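# Physics asset for the FaceAcc part of the female medium EmberYearnHelmet cosmetic, defining capsule bodies and constraints for its dangling hair bones.
# Four short dyn_Hair chains (left, left-mid, right, front-mid right) each constrain to the head, then link bone to bone.
# Swing1 limits shrink toward the tips, with narrow swing2 and zero twist limits throughout.
# Unconstrained capsules on head, neck, upper spine, clavicles and upper arms provide body collision shapes.
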PhysicsAsset /BRCosmetics/Characters/Player/Female/Medium/Bodies/F_MED_EmberYearnHelmet/Meshes/Parts/F_MED_EmberYearnHelmet_FaceAcc_FH_Physics
bodies {
 clavicle_l [{
   shapeType=Capsule
   length=11.5873
   radius=6.8082
   shapeName=clavicle_l_capsule
   center=<4.3097 -1.3675 -1.8832>
   rotation=<rot -0.3852 -86.854 -76.619>
  }]
 clavicle_r [{
   shapeType=Capsule
   length=12.3956
   radius=5.9999
   shapeName=clavicle_r_capsule
   center=<-4.3097 1.3675 1.8832>
   rotation=<rot -0.3852 -86.854 103.381>
  }]
 dyn_Hair_03_l [{
   shapeType=Capsule
   length=3.167
   radius=1.7037
   shapeName=dyn_Hair_03_l_capsule
   center=<2.2823 -0.6659 -1.6099>
   rotation=<rot 15.8844 84.6797 -58.6363>
  }]
 dyn_Hair_03_r [{
   shapeType=Capsule
   length=5.7138
   radius=1.5192
   shapeName=dyn_Hair_03_r_capsule
   center=<2.3747 1.3531 -0.6228>
   rotation=<rot 37.282 -115.022 -104.9466>
  }]
 dyn_Hair_04_l [{
   shapeType=Capsule
   length=3.7684
   radius=1.4056
   shapeName=dyn_Hair_04_l_capsule
   center=<1.712 -0.0858 0.2247>
   rotation=<rot 1.845 102.1327 -58.7774>
  }]
 dyn_Hair_04_r [{
   shapeType=Capsule
   length=4.8639
   radius=1.2552
   shapeName=dyn_Hair_04_r_capsule
   center=<1.7425 0.011 -0.0502>
   rotation=<rot -0.0625 74.4874 -90.6711>
  }]
 dyn_Hair_front_mid_02_r [{
   shapeType=Capsule
   length=4.4468
   radius=1.6858
   shapeName=dyn_Hair_front_mid_02_r_capsule
   center=<1.4711 -0.6588 -1.1829>
   rotation=<rot -37.9974 -70.012 -102.1564>
  }]
 dyn_Hair_front_mid_03_r [{
   shapeType=Capsule
   length=3.6232
   radius=1.5984
   shapeName=dyn_Hair_front_mid_03_r_capsule
   center=<1.5824 -0.6511 0.4023>
   rotation=<rot 3.6512 82.1962 -58.275>
  }]
 dyn_Hair_mid_03_l [{
   shapeType=Capsule
   length=1.744
   radius=1.6417
   shapeName=dyn_Hair_mid_03_l_capsule
   center=<2.7506 -0.0851 -1.515>
   rotation=<rot 7.2591 -73.418 -112.9941>
  }]
 dyn_Hair_mid_04_l [{
   shapeType=Capsule
   length=2.3516
   radius=1.7642
   shapeName=dyn_Hair_mid_04_l_capsule
   center=<2.642 -0.276 -0.007>
   rotation=<rot -0.0003 85.3921 -93.8853>
  }]
 dyn_Hair_mid_05_l [{
   shapeType=Capsule
   length=2.8338
   radius=1.3509
   shapeName=dyn_Hair_mid_05_l_capsule
   center=<1.7717 0.312 -1.181>
   rotation=<rot 0.9255 -65.0175 -63.1641>
  }]
 head [{shapeType=Capsule, length=4.5214, radius=7.1602, shapeName=head_capsule, center=<4.0392 3.187 -0.0072>, rotation=<rot -90 115.6048 0>}]
 neck_01 [{
   shapeType=Capsule
   length=6.7545
   radius=4.4992
   shapeName=neck_01_capsule
   center=<2.596 0.5343 -0.3833>
   rotation=<rot 9.5548 -108.4166 -90.5696>
  }]
 spine_05 [
  {shapeType=Capsule, length=20.7869, radius=7.7922, shapeName=spine_05_capsule, center=<1.2036 -3.3459 0>, rotation=<rot 0 -90 0>}
  {shapeType=Capsule, length=12.8564, radius=10.2325, shapeName=spine_05_capsule, center=<-4.7366 1.4158 0>}
 ]
 upperarm_l [{
   shapeType=Capsule
   length=18.0196
   radius=5.2401
   shapeName=upperarm_l_capsule
   center=<10.4918 -1.8629 -1.0638>
   rotation=<rot -0.7762 -91.0156 -90.2084>
  }]
 upperarm_r [{
   shapeType=Capsule
   length=17.989
   radius=4.5207
   shapeName=upperarm_r_capsule
   center=<-10.4922 1.8613 1.176>
   rotation=<rot -0.7762 -91.0156 89.7916>
  }]
}
constraints [
 {bone1Name=dyn_Hair_front_mid_02_r, bone2Name=head, swing1LimitDegrees=60, swing2LimitDegrees=5, twistLimitDegrees=0}
 {bone1Name=dyn_Hair_front_mid_03_r, bone2Name=dyn_Hair_front_mid_02_r, swing1LimitDegrees=30, swing2LimitDegrees=5, twistLimitDegrees=0}
 {bone1Name=dyn_Hair_03_r, bone2Name=head, swing1LimitDegrees=45, swing2LimitDegrees=5, twistLimitDegrees=0}
 {bone1Name=dyn_Hair_04_r, bone2Name=dyn_Hair_03_r, swing1LimitDegrees=10, swing2LimitDegrees=5, twistLimitDegrees=0}
 {bone1Name=dyn_Hair_mid_03_l, bone2Name=head, swing1LimitDegrees=20, swing2LimitDegrees=5, twistLimitDegrees=0}
 {bone1Name=dyn_Hair_mid_04_l, bone2Name=dyn_Hair_mid_03_l, swing1LimitDegrees=15, swing2LimitDegrees=5, twistLimitDegrees=0}
 {bone1Name=dyn_Hair_mid_05_l, bone2Name=dyn_Hair_mid_04_l, swing1LimitDegrees=10, swing2LimitDegrees=5, twistLimitDegrees=0}
 {bone1Name=dyn_Hair_03_l, bone2Name=head, swing1LimitDegrees=45, swing2LimitDegrees=5, twistLimitDegrees=0}
 {bone1Name=dyn_Hair_04_l, bone2Name=dyn_Hair_03_l, swing1LimitDegrees=30, swing2LimitDegrees=5, twistLimitDegrees=0}
]
