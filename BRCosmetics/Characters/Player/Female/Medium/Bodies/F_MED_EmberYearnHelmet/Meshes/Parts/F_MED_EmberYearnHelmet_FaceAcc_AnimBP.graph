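# Post-process style AnimBP for the EmberYearnHelmet female face accessory, adding secondary motion to its hair bones.
# The parent mesh pose feeds a Control Rig, chained Anim Dynamics nodes for back and front hair, then a corrective Control Rig.
# EventGraph derives that rig's inputs from head rotation, aim, emote, melee and skydive states, and records backbling flags on customization load.
# A RigidBody node sits unconnected to the output.
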AnimBlueprint /BRCosmetics/Characters/Player/Female/Medium/Bodies/F_MED_EmberYearnHelmet/Meshes/Parts/F_MED_EmberYearnHelmet_FaceAcc_AnimBP
target_skeleton=&/F_MED_EmberYearnHelmet_FaceAcc_Skeleton
graphs [
 AnimGraph {nodes [
   RigidBody!(
    ComponentPose
    Alpha
    OverrideWorldGravity:Vector="0,0,0"
    ExternalForce
    SimSpaceSettings:Sim Space Settings Structure=(WorldAlpha=0,VelocityScaleZ=1,DampingAlpha=1,MaxLinearVelocity=10000,MaxAngularVelocity=10000,MaxLinearAcceleration=10000,MaxAngularAcceleration=10000,ExternalLinearDragV=(X=0,Y=0,Z=0),ExternalLinearVelocity=(X=0,Y=0,Z=0),ExternalAngularVelocity=(X=0,Y=0,Z=0))
   ) -> (Pose) {
    bOverrideWorldGravity=true
    overridePhysicsAsset=&/F_MED_EmberYearnHelmet_FaceAcc_FH_Physics
    simulationSpace=BaseBoneSpace
    baseBoneRef {boneName=spine_05}
    componentAppliedLinearAccClamp=<2000 2000 2000>
   }
   n1 = Control Rig(Source=Copy Pose From Mesh() {bUseAttachedParent=true}) {controlRigAssetReference {blueprintRigClass=&/F_MED_EmberYearnHelmet_FaceAcc_CtrlRig.F_MED_EmberYearnHelmet_FaceAcc_CtrlRig_C}}
   n2 = Anim Dynamics - Bone: dyn_Hair_BackL_03!(ComponentPose=Local To Component(LocalPose=@n1), bAlphaBoolEnabled:bool=True, GravityOverride:Vector="0,0,-980", ExternalForce="0,0,0") -> (Pose) {
    alphaInputType=Bool
    angularDampingOverride=1
    angularSpringConstant=8000
    bAngularSpring=true
    bOverrideAngularDamping=true
    bUseGravityOverride=true
    bUseSphericalLimits=true
    gravityScale=0.35
    simulationSpace=RootRelative
    boundBone {boneName=dyn_Hair_BackL_03}
    componentAppliedLinearAccClamp=<2000 2000 2000>
    componentLinearAccScale=<1 1 1>
    componentLinearVelScale=<1 1 1>
    gravityOverride {z=-980}
    physicsBodyDefinitions [{
      boundBone {boneName=dyn_Hair_BackL_03}
      boxExtents=<30 30 30>
      constraintSetup {angularLimitsMax=<0 -4 5>, angularLimitsMin=<0 -40 -5>, angularTarget=<1 0 0>}
      localJointOffset=<4 0 0>
     }]
    sphericalLimits [{limitRadius=8, drivingBone {boneName=clavicle_l}, sphereLocalOffset=<21.5199 -5.8944 6.9672>}]
   }
   n3 = Anim Dynamics - Bone: dyn_Hair_Back_03!(ComponentPose=@n2, bAlphaBoolEnabled:bool=True, GravityOverride:Vector="0,0,-980", ExternalForce="0,0,0") -> (Pose) {
    alphaInputType=Bool
    angularDampingOverride=1
    angularSpringConstant=8000
    bAngularSpring=true
    bOverrideAngularDamping=true
    bUseGravityOverride=true
    gravityScale=0.35
    simulationSpace=RootRelative
    boundBone {boneName=dyn_Hair_Back_03}
    componentAppliedLinearAccClamp=<2000 2000 2000>
    componentLinearAccScale=<1 1 1>
    componentLinearVelScale=<1 1 1>
    gravityOverride {z=-980}
    physicsBodyDefinitions [{
      boundBone {boneName=dyn_Hair_Back_03}
      boxExtents=<30 30 30>
      constraintSetup {angularLimitsMax=<0 -4 5>, angularLimitsMin=<0 -40 -5>, angularTarget=<1 0 0>}
      localJointOffset=<4 0 0>
     }]
   }
   n4 = Anim Dynamics - Bone: dyn_Hair_BackR_03!(ComponentPose=@n3, bAlphaBoolEnabled:bool=True, GravityOverride:Vector="0,0,-980", ExternalForce="0,0,0") -> (Pose) {
    alphaInputType=Bool
    angularDampingOverride=1
    angularSpringConstant=8000
    bAngularSpring=true
    bOverrideAngularDamping=true
    bUseGravityOverride=true
    gravityScale=0.35
    simulationSpace=RootRelative
    boundBone {boneName=dyn_Hair_BackR_03}
    componentAppliedLinearAccClamp=<2000 2000 2000>
    componentLinearAccScale=<1 1 1>
    componentLinearVelScale=<1 1 1>
    gravityOverride {z=-980}
    physicsBodyDefinitions [{
      boundBone {boneName=dyn_Hair_BackR_03}
      boxExtents=<30 30 30>
      constraintSetup {angularLimitsMax=<0 -4 5>, angularLimitsMin=<0 -40 -5>, angularTarget=<1 0 0>}
      localJointOffset=<4 0 0>
     }]
    sphericalLimits [{limitRadius=9, drivingBone {boneName=upperarm_r}, sphereLocalOffset=<0 -12 60>}]
   }
   n5 = Anim Dynamics - Bone: dyn_Hair_BackL_04!(ComponentPose=@n4, bAlphaBoolEnabled:bool=True, GravityOverride:Vector="0,0,-980", ExternalForce="0,0,0") -> (Pose) {
    alphaInputType=Bool
    angularDampingOverride=0.9999
    angularSpringConstant=8000
    bAngularSpring=true
    bOverrideAngularDamping=true
    bUseGravityOverride=true
    gravityScale=0.35
    simulationSpace=RootRelative
    boundBone {boneName=dyn_Hair_BackL_04}
    componentAppliedLinearAccClamp=<4000 4000 4000>
    componentLinearAccScale=<1 1 1>
    componentLinearVelScale=<1 1 1>
    gravityOverride {z=-980}
    physicsBodyDefinitions [{
      boundBone {boneName=dyn_Hair_BackL_04}
      boxExtents=<25 20 20>
      constraintSetup {angularLimitsMax=<0 -2 5>, angularLimitsMin=<0 -45 -5>, angularTarget=<1 0 0>}
      localJointOffset=<4 0 0>
     }]
   }
   n6 = Anim Dynamics - Bone: dyn_Hair_Back_04!(ComponentPose=@n5, bAlphaBoolEnabled:bool=True, GravityOverride:Vector="0,0,-980", ExternalForce="0,0,0") -> (Pose) {
    alphaInputType=Bool
    angularDampingOverride=0.9999
    angularSpringConstant=8000
    bAngularSpring=true
    bOverrideAngularDamping=true
    bUseGravityOverride=true
    gravityScale=0.35
    simulationSpace=RootRelative
    boundBone {boneName=dyn_Hair_Back_04}
    componentAppliedLinearAccClamp=<4000 4000 4000>
    componentLinearAccScale=<1 1 1>
    componentLinearVelScale=<1 1 1>
    gravityOverride {z=-980}
    physicsBodyDefinitions [{
      boundBone {boneName=dyn_Hair_Back_04}
      boxExtents=<25 20 20>
      constraintSetup {angularLimitsMax=<0 -2 5>, angularLimitsMin=<0 -45 -5>, angularTarget=<1 0 0>}
      localJointOffset=<4 0 0>
     }]
   }
   n7 = Anim Dynamics - Bone: dyn_Hair_BackR_04!(ComponentPose=@n6, bAlphaBoolEnabled:bool=True, GravityOverride:Vector="0,0,-980", ExternalForce="0,0,0") -> (Pose) {
    alphaInputType=Bool
    angularDampingOverride=0.9999
    angularSpringConstant=8000
    bAngularSpring=true
    bOverrideAngularDamping=true
    bUseGravityOverride=true
    gravityScale=0.35
    simulationSpace=RootRelative
    boundBone {boneName=dyn_Hair_BackR_04}
    componentAppliedLinearAccClamp=<4000 4000 4000>
    componentLinearAccScale=<1 1 1>
    componentLinearVelScale=<1 1 1>
    gravityOverride {z=-980}
    physicsBodyDefinitions [{
      boundBone {boneName=dyn_Hair_BackR_04}
      boxExtents=<25 20 20>
      constraintSetup {angularLimitsMax=<0 -2 5>, angularLimitsMin=<0 -45 -5>, angularTarget=<1 0 0>}
      localJointOffset=<4 0 0>
     }]
   }
   n8 = Anim Dynamics - Bone: dyn_Hair_BackL_05!(ComponentPose=@n7, bAlphaBoolEnabled:bool=True, GravityOverride:Vector="0,0,-980", ExternalForce="0,0,0") -> (Pose) {
    alphaInputType=Bool
    angularDampingOverride=0.999
    angularSpringConstant=5000
    bAngularSpring=true
    bOverrideAngularDamping=true
    bUseGravityOverride=true
    gravityScale=0.35
    simulationSpace=RootRelative
    boundBone {boneName=dyn_Hair_BackL_05}
    componentAppliedLinearAccClamp=<6000 6000 6000>
    componentLinearAccScale=<1 1 1>
    componentLinearVelScale=<1 1 1>
    gravityOverride {z=-980}
    physicsBodyDefinitions [{
      boundBone {boneName=dyn_Hair_BackL_05}
      boxExtents=<30 30 30>
      constraintSetup {angularLimitsMax=<0 -2 0>, angularLimitsMin=<0 -60 0>, angularTarget=<1 0 0>}
      localJointOffset=<4 0 0>
     }]
   }
   n9 = Anim Dynamics - Bone: dyn_Hair_Back_05!(ComponentPose=@n8, bAlphaBoolEnabled:bool=True, GravityOverride:Vector="0,0,-980", ExternalForce="0,0,0") -> (Pose) {
    alphaInputType=Bool
    angularDampingOverride=0.999
    angularSpringConstant=5000
    bAngularSpring=true
    bOverrideAngularDamping=true
    bUseGravityOverride=true
    gravityScale=0.35
    simulationSpace=RootRelative
    boundBone {boneName=dyn_Hair_Back_05}
    componentAppliedLinearAccClamp=<6000 6000 6000>
    componentLinearAccScale=<1 1 1>
    componentLinearVelScale=<1 1 1>
    gravityOverride {z=-980}
    physicsBodyDefinitions [{
      boundBone {boneName=dyn_Hair_Back_05}
      boxExtents=<30 30 30>
      constraintSetup {angularLimitsMax=<0 -2 0>, angularLimitsMin=<0 -60 0>, angularTarget=<1 0 0>}
      localJointOffset=<4 0 0>
     }]
   }
   n10 = Anim Dynamics - Bone: dyn_Hair_BackR_05!(ComponentPose=@n9, bAlphaBoolEnabled:bool=True, GravityOverride:Vector="0,0,-980", ExternalForce="0,0,0") -> (Pose) {
    alphaInputType=Bool
    angularDampingOverride=0.999
    angularSpringConstant=5000
    bAngularSpring=true
    bOverrideAngularDamping=true
    bUseGravityOverride=true
    gravityScale=0.35
    simulationSpace=RootRelative
    boundBone {boneName=dyn_Hair_BackR_05}
    componentAppliedLinearAccClamp=<6000 6000 6000>
    componentLinearAccScale=<1 1 1>
    componentLinearVelScale=<1 1 1>
    gravityOverride {z=-980}
    physicsBodyDefinitions [{
      boundBone {boneName=dyn_Hair_BackR_05}
      boxExtents=<30 30 30>
      constraintSetup {angularLimitsMax=<0 -2 0>, angularLimitsMin=<0 -60 0>, angularTarget=<1 0 0>}
      localJointOffset=<4 0 0>
     }]
   }
   n11 = Anim Dynamics - Bone: dyn_front_hair_r_01!(SimulationSpace:AnimPhysSimSpaceType Enum=World, ComponentPose=@n10, Alpha=1, ExternalForce="0,0,0") -> (Pose) {
    angularDampingOverride=1
    angularSpringConstant=20000
    bAngularSpring=true
    bOverrideAngularDamping=true
    bUseGravityOverride=true
    gravityScale=0.35
    simulationSpace=World
    boundBone {boneName=dyn_front_hair_r_01}
    componentLinearAccScale=<1 1 1>
    componentLinearVelScale=<1 1 1>
    gravityOverride {z=-980}
    physicsBodyDefinitions [{
      boundBone {boneName=dyn_front_hair_r_01}
      boxExtents=<20 30 30>
      constraintSetup {angularLimitsMax=<15 15 15>, angularLimitsMin=<-15 -15 -15>, angularTarget=<1 0 0>}
      localJointOffset=<4 0 0>
     }]
   }
   n12 = Anim Dynamics - Bone: dyn_front_hair_l_01!(SimulationSpace:AnimPhysSimSpaceType Enum=World, ComponentPose=@n11, Alpha=1, ExternalForce="0,0,0") -> (Pose) {
    angularDampingOverride=1
    angularSpringConstant=20000
    bAngularSpring=true
    bOverrideAngularDamping=true
    bUseGravityOverride=true
    gravityScale=0.35
    simulationSpace=World
    boundBone {boneName=dyn_front_hair_l_01}
    componentLinearAccScale=<1 1 1>
    componentLinearVelScale=<1 1 1>
    gravityOverride {z=-980}
    physicsBodyDefinitions [{
      boundBone {boneName=dyn_front_hair_l_01}
      boxExtents=<20 30 30>
      constraintSetup {angularLimitsMax=<15 15 15>, angularLimitsMin=<-15 -15 -15>, angularTarget=<1 0 0>}
      localJointOffset=<4 0 0>
     }]
   }
   n13 = Control Rig!(
    Source=Component To Local(ComponentPose=@n12)
    Alpha
    Melee_Float:double=Get Melee_Float() -> (Melee_Float:double)
    CrouchAim_Up:double=Get Is_CrouchAim_HeadUp() -> (Is_CrouchAim_HeadUp:double)
    StandAim_HeadUp:double=Get Is_StandAim_HeadUp() -> (Is_StandAim_HeadUp:double)
    LeftRot:double=Get HeadLeftFloat() -> (HeadLeftFloat:double)
    Emote_head_rot:double=Get Emote_left() -> (Emote_left:double)
    RightRot:double=Get Emote_right() -> (Emote_right:double)
    Head_up:double=Get Emote_up() -> (Emote_up:double)
    is_Emote:double=Get is_Emote() -> (is_Emote:double)
    Is_Parachute:double=Get is_Parachute() -> (is_Parachute:double)
    Is_skydivefloat:double=Get is_Skydivefloat() -> (is_Skydivefloat:double)
    HeadUp_R:double=Get HeadUp_R() -> (HeadUp_R:double)
   ) -> (Pose) {controlRigAssetReference {blueprintRigClass=&/F_MED_EmberYearnHelmet_Post_FaceAcc_CtrlRig.F_MED_EmberYearnHelmet_Post_FaceAcc_CtrlRig_C}}
   Output Pose(Result=@n13) {layerGroup=None}
  ]}
 EventGraph {nodes [
   Event BlueprintUpdateAnimation()
   > n1 = Sequence!(execute) -> (then_0, then_1, then_2:Exec)
   n2 = Get bIsCrouching()
   n3 = Get bIsTargeting()
   Set Is_CrouchAim_HeadUp(execute=@n1.then_2, Is_CrouchAim_HeadUp:double="To Float (Boolean)"(InBool=AND Boolean(A=@n3, B=@n2))) -> (Output_Get:double)
   n4 = NOT Boolean(A=Get bIsCrouching())
   n5 = Get bIsTargeting()
   > Set Is_StandAim_HeadUp(Is_StandAim_HeadUp:double="To Float (Boolean)"(InBool=AND Boolean(A=@n5, B=@n4))) -> (Output_Get:double)
   > Set is_Emote(is_Emote:double="To Float (Boolean)"(InBool=Get bIsPlayingEmote())) -> (Output_Get:double)
   > Set is_Parachute(is_Parachute:double="To Float (Boolean)"(InBool=Get bIsParachuteOpened())) -> (Output_Get:double)
   > Set is_Skydivefloat(is_Skydivefloat:double="To Float (Boolean)"(InBool=Get bIsSkydiveFloating())) -> (Output_Get:double)
   > Set HeadUp_R(HeadUp_R:double=Get HeadRightFloat() -> (HeadRightFloat:double)) -> (Output_Get:double)
   > n6 = Branch(Condition=Get bIsInFrontEnd())
   > Set Sim Space Anim(Sim Space Anim:AnimPhysSimSpaceType Enum=World) -> (Output_Get:AnimPhysSimSpaceType Enum)
   > Set AnimDynAlpha(AnimDynAlpha:double=0.4) -> (Output_Get:double)
   Set Sim Space Anim(execute=@n6.else)
   > Set AnimDynAlpha(AnimDynAlpha=1)
   n7 = GetOwnerSkeletalMeshComponent()
   Set HeadLeftFloat(
    execute=@n1.then_1
    HeadLeftFloat:double=GetAlphaFromReferenceTransformDelta(SkeletalMeshComponent=@n7, BoneName=head, BaseName=spine_05, DeltaComponentType=EulerX, InputRange=(X=20,Y=75), TwistAxis="1,0,0")
   ) -> (Output_Get:double)
   > Set HeadRightFloat(HeadRightFloat:double=GetAlphaFromReferenceTransformDelta(SkeletalMeshComponent=@n7, BoneName=head, BaseName=spine_05, DeltaComponentType=EulerX, InputRange=(X=-20,Y=-70), TwistAxis="1,0,0")) -> (Output_Get:double)
   > Set HeadDownFloat(HeadDownFloat:double=GetAlphaFromReferenceTransformDelta(SkeletalMeshComponent=@n7, BoneName=head, BaseName=spine_05, InputRange=(X=0,Y=18), TwistAxis="1,0,0")) -> (Output_Get:double)
   n8 = GetAlphaFromReferenceTransformDelta(SkeletalMeshComponent=@n7, BoneName=head, BaseName=spine_05, InputRange=(X=0,Y=-50), TwistAxis="1,0,0")
   > Set HeadUpFloat(HeadUpFloat:double=@n8) -> (Output_Get:double)
   > Set is_HeadUp_only(is_HeadUp_only:double=@n8) -> (Output_Get:double)
   > Set spineDeltaFloat(spineDeltaFloat:double=GetAlphaFromReferenceTransformDelta(SkeletalMeshComponent=@n7, BoneName=i, BaseName=spine_01, InputRange=(X=10,Y=15), TwistAxis="1,0,0")) -> (Output_Get:double)
   > Set Emote_left(Emote_left:double=SelectFloat(A=Get HeadLeftFloat(), B=0, bPickA=Get bIsPlayingEmote())) -> (Output_Get:double)
   > n9 = Branch(Condition=Get bIsPlayingEmote())
   > Set Emote_right(Emote_right:double=Get HeadRightFloat()) -> (Output_Get:double)
   > Set Emote_up(Emote_up:double=Get HeadUpFloat() -> (HeadUpFloat:double)) -> (Output_Get:double)
   Set Emote_right(execute=@n9.else, Emote_right=0)
   > Set Emote_up(Emote_up=0)
   n10 = "Gravity Override State Function (Parameter Asset) (Nativized)"(
    execute=@n1.then_0
    CharacterPartAnimInstance=Self-Reference()
    GameTimeInSeconds=GetGameTimeInSeconds()
    Parameters=&/F_MED_EmberYearnHelmet_FaceAcc_Parameters
   )
   > Set GravityOverride(GravityOverride=@n10)
   > Set RBAN Settings Out(RBAN Settings Out=@n10.OutRigidBodyAnimNodeParameters)
   > Set Melee_Float(Melee_Float:double="To Float (Boolean)"(InBool=Get bIsPlayingMeleeAnim())) -> (Output_Get:double)
   Event OnCharacterCustomizationLoaded()
   n11 = CheckForBackbling()
   > Set isWearingBackbling(isWearingBackbling=@n11)
   > Set isWearingB_C(isWearingB_C:bool=NOT Boolean(A=OR Boolean!(A=@n11, B=@n11.bIsBackblingACape, C:bool=@n11.bIsBackblingLong) -> (ReturnValue))) -> (Output_Get:bool)
   n12 = OR Boolean!(A=@n11.bIsBackblingACape, B=@n11.bIsBackblingLong) -> (ReturnValue)
   > Set is_WearingCape(is_WearingCape:bool=@n12) -> (Output_Get:bool)
   > Set is_f_WearingCape(is_f_WearingCape:double="To Float (Boolean)"(InBool=@n12)) -> (Output_Get:double)
   > Set isWearingBP(isWearingBP:bool=AND Boolean(A=@n11, B=NOT Boolean(A=@n12))) -> (Output_Get:bool)
   > Set Backbling(Backbling:double="To Float (Boolean)"(InBool=OR Boolean!(A=@n11, B=@n11.bIsBackblingACape, C:bool=@n11.bIsBackblingLong) -> (ReturnValue))) -> (Output_Get:double)
  ]}
]
variables [
 AnimDynAlpha {type=number}
 Backbling {type=number}
 Emote_left {type=number}
 Emote_right {type=number}
 Emote_up {type=number}
 GravityOverride {type=Vector}
 HeadDownFloat {type=number}
 HeadLeftFloat {type=number}
 HeadRightFloat {type=number}
 HeadUp_R {type=number}
 HeadUpFloat {type=number}
 Is_CrouchAim_HeadUp {type=number}
 is_Emote {type=number}
 is_f_WearingCape {type=number}
 is_HeadUp_only {type=number}
 is_Parachute {type=number}
 is_Skydivefloat {type=number}
 Is_StandAim_HeadUp {type=number}
 is_WearingCape {type=boolean}
 isWearingB_C {type=boolean}
 isWearingBackbling {type=boolean}
 isWearingBP {type=boolean}
 Melee_Float {type=number}
 RBAN Settings Out {type=FortRigidBodyAnimNodeParameters}
 Sim Space Anim {type=AnimPhysSimSpaceType}
 SkelComponent {type=SkeletalMeshComponent}
 Skeletal_Mesh {type=SkeletalMeshComponent}
 spineDeltaFloat {type=number}
]
